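; Objective marker actor that shows a world ping on a SonicSpear zone, triggered by the Cloud Finished Forming event.
; On cosmetic-capable machines it checks each local player controller, skipping players tagged as already inside a zone.
; It registers the ping only when the player is within a scalable 2D distance and its owner is the zone actor closest to their view target.
; It first forces the markable state in case replication lags, and does not control the minimap mark.
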(event Custom|Cloud%20Finished%20Forming
  ;[0 0.1238 0.15 0.5] Call the marker on all relative clients
  (CallFunction|MarkSonicSpearObjective))

(fn MarkSonicSpearObjective
  (Networking|CanExecuteCosmeticEvents
    (:True
      (bind _outactors (Actor|GetAllActorsOfClass "/SonicSpearGameplay/Gameplay/BP_SonicSpearZoneActor.BP_SonicSpearZoneActor_C"))
      (for _array_element (Fortnite|GetLocalFortPlayerControllers true)
        (Utilities|IsValid _array_element
          (:Is%20Valid
            (bind _returnvalue (GameplayTags|HasMatchingGameplayTag(Message) _array_element (Variables|Default|GetSonicSpearInZoneTagToCheck))) ; check if the player has the SonicSpear.Zone.Inside tag\nPlayers already inside the zone shouldn't get the standard \nobjective marker, but should have the UI toast instead
            (if (Math|Boolean|ANDBoolean (not _returnvalue) (< (Math|Vector|Distance2DSquared(Vector) (Transformation|GetActorLocation self) (Transformation|GetActorLocation (Game|Player|GetPlayerPawn _array_element))) (Math|Float|Square (ScalableFloat|GetValueatLevel (Variables|Default|GetRow_MaxDistanceToSeeObjectivePing)))) (== (Fortnite|GetClosestActorfromArray (Pawn|GetViewTarget _array_element) _outactors) (Actor|GetOwner)))
              (MarkableActor|SetCanBeMarked true) ; Ensure the actor can be marked in case \nits state variable hasn't replicated yet.
              (Fortnite|RegisterSoundIndicatorPlacementContext _array_element self "0, 0, 0" false) ; Though strangely named, this node marks the actor \nwith the world UI ping for clients\n-Only want to show 1 of these, marking the closest zone to the player\n-This does not determine minimap mark.
              (Development|PrintString "ping donut marker"))))))) ; objective marking only can occur on the client
)

(fn ConstructionScript)
